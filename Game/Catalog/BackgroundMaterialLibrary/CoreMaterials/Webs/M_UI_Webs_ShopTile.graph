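# Translucent UI material that draws an animated spider web background for item shop tiles.
# Three web texture channels get separate scale and offset, noise-driven flow-map distortion, and either opposing pans (red, green) or seeded rotation (blue).
# Weighted Max operations combine web masks, alongside an angled gradient and background color parameters from a shared function.
# Rounded-corner functions and an SDF mask, using the shop's CornerRadius collection parameter, supply tile shape and soft outline frame.
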Material /Game/Catalog/BackgroundMaterialLibrary/CoreMaterials/Webs/M_UI_Webs_ShopTile
graph {
 properties {bAutomaticallySetUsageInEditor=false, blendMode=BLEND_Translucent, materialDomain=MD_UI, usedShadingModels=Default Lit}
 nodes [
  n1 = MaterialFunctionCall!(Input Alpha=Constant() {r=1}, bOverrideCornerRadius, Corner Radius Override) {materialFunction=/FortUILibrary/Systems/ItemShopTile/Materials/MaterialFunctions/MF_UI_ItemShop_RoundedCorners}
  Add(A=NamedRerouteUsage(), B=LinearInterpolate(A=NamedRerouteUsage(), B=@n1.Final Alpha, Alpha=NamedRerouteUsage()) {constA=1})
  Constant3Vector() {constant=<rgba 0.0244 0.0231 0.0625 1>}
  Constant3Vector() {constant=<rgba 0.3514 0.3432 0.599 1>}
  Min(A=@n1.Final Alpha, B=Add(A=NamedRerouteUsage()) {constB=0.5})
  n2 = CollectionParameter() {collection=/FortUILibrary/Systems/ItemShopTile/MPC_ItemShop, parameterName=CornerRadius}
  n3 = MaterialFunctionCall(
   Corner Radius=MaterialFunctionCall(R=@n2, G=@n2, B=@n2, A=@n2) {materialFunction=&AppendMany}.RGBA
   Radius in Pixels=Constant() {r=1}
   Outline Thickness=Constant() {r=4}
  ) {materialFunction=&MF_UI_CoreUI_RoundedMask}
  n4 = OneMinus(None=SmoothStep(Value=@n3.Outline SDF) {constMax=0.1, constMin=-0.2})
  Multiply(A=@n4)
  NamedRerouteDeclaration(None=TextureObject() {texture=/Game/Catalog/BackgroundMaterialLibrary/Textures/T_UI_WebTexture}) {name=Web_Texture, nodeColor=<rgba 1 0 0.0938 1>}
  NamedRerouteDeclaration(None=@n3.Outline Mask) {name=ShapeMask, nodeColor=<rgba 1 0.8906 0 1>}
  n5 = MaterialFunctionCall(UV=NamedRerouteDeclaration(None=MaterialFunctionCall() {materialFunction=&GetUserInterfaceUV}.Normalized UV) {name=Normalized UV, nodeColor=<rgba 0 0 0 1>}) {materialFunction=&MF_UI_UVScaling}
  NamedRerouteDeclaration(None=@n5.Scaled To Fill) {name=UVFit, nodeColor=<rgba 1 0 0.9844 1>}
  NamedRerouteDeclaration(None=@n5.ScaleX) {name=UVX, nodeColor=<rgba 0 1 0.6953 1>}
  NamedRerouteDeclaration(None=@n5.ScaleY) {name=UVY, nodeColor=<rgba 1 0.375 0 1>}
  n6 = NamedRerouteUsage()
  n7 = NamedRerouteUsage()
  n8 = MaterialFunctionCall(UVs=NamedRerouteUsage(), GradientStart=Constant() {r=0.15}, GradientEnd=Constant() {r=1}, RotationAngle=Constant() {r=0.25}) {materialFunction=&MF_UI_LinearGradient}
  NamedRerouteDeclaration(None=Divide(A=LinearInterpolate(A=@n6, B=@n7, Alpha=@n8.Result)) {constB=0.8}) {name=BG Gradient, nodeColor=<rgba 0.875 1 0 1>}
  n9 = NamedRerouteUsage()
  n10 = Multiply(A=NamedRerouteUsage()) {constB=1.5}
  n11 = Multiply(A=NamedRerouteUsage()) {constB=0.4}
  NamedRerouteDeclaration(None=Max(A=Multiply(A=@n9) {constB=0.2}, B=Max(A=@n10, B=@n11))) {name=CombinedWebMask, nodeColor=<rgba 0.0547 0 1 1>}
  n12 = NamedRerouteUsage()
  n13 = ScalarParameter() {defaultValue=1, group=Red, parameterName=RedScale}
  n14 = ScalarParameter() {group=Red, parameterName=RedX}
  n15 = ScalarParameter() {group=Red, parameterName=RedY}
  n16 = Panner(Coordinate=Add(
    A=MaterialFunctionCall(UVs=@n12, Texture Scale=@n13) {materialFunction=&ScaleUVsByCenter}.UVs
    B=AppendVector(A=@n14, B=@n15)
   )) {speedX=-0.015, speedY=-0.015}
  n17 = NamedRerouteUsage()
  n18 = TextureSample(UVs=MaterialFunctionCall(UVs=NamedRerouteUsage(), Texture Scale=Constant() {r=2}) {materialFunction=&ScaleUVsByCenter}.UVs) {samplerType=SAMPLERTYPE_Grayscale, texture=/Game/UIKit/Buttons/Textures/LowResBlurredNoise}
  n19 = Time()
  n20 = Multiply(A=@n19) {constB=0.25}
  n21 = Constant() {r=0.02}
  n22 = Constant() {r=0.04}
  NamedRerouteDeclaration(None=ComponentMask(None=MaterialFunctionCall!(UVs=@n16, Texture Object=@n17, Grayscale Flow Map=@n18.RGB, Flow Direction=@n20, Flow Intensity=@n21, Flow Speed=@n22) {materialFunction=/FortUILibrary/MaterialFunctions/Effects/MF_UI_GrayscaleFlowMap_New}.Distorted Texture) {r=true}) {name=Red, nodeColor=<rgba 0.2891 1 0 1>}
  n23 = ScalarParameter() {defaultValue=1, group=Green, parameterName=GreenScale}
  n24 = ScalarParameter() {group=Green, parameterName=GreenX}
  n25 = ScalarParameter() {group=Green, parameterName=GreenY}
  n26 = Panner(Coordinate=Add(
    A=MaterialFunctionCall(UVs=@n12, Texture Scale=@n23) {materialFunction=&ScaleUVsByCenter}.UVs
    B=AppendVector(A=@n24, B=@n25)
   )) {speedX=0.02, speedY=0.02}
  n27 = NamedRerouteUsage()
  n28 = Constant() {r=0.055}
  n29 = Constant() {r=0.25}
  NamedRerouteDeclaration(None=ComponentMask(None=MaterialFunctionCall(UVs=@n26, Texture Object=@n27, Grayscale Flow Map=@n18.RGB, Flow Direction=@n20, Flow Intensity=@n28, Flow Speed=@n29) {materialFunction=&MF_UI_GrayscaleFlowMap_New}.Distorted Texture) {g=true}) {name=Green, nodeColor=<rgba 1 0 0.6094 1>}
  NamedRerouteDeclaration(None=@n4) {name=OutlineFrame, nodeColor=<rgba 0 0.5547 1 1>}
  n30 = MaterialFunctionCall() {materialFunction=/FortUILibrary/Systems/ItemShopTile/Materials/MaterialFunctions/MF_UI_ItemShop_BackgroundParams}
  NamedRerouteDeclaration(None=@n30.Color 3) {name=Accent Color, nodeColor=<rgba 0 0 0 1>}
  NamedRerouteDeclaration(None=@n30.Color 1) {name=Primary Color, nodeColor=<rgba 0 0 0 1>}
  NamedRerouteDeclaration(None=@n30.Color 2) {name=Secondary Color, nodeColor=<rgba 0 0 0 1>}
  n31 = NamedRerouteUsage()
  n32 = Multiply(A=@n19) {constB=-0.15}
  n33 = Constant() {r=0.015}
  n34 = Constant() {r=0.04}
  n35 = NamedRerouteUsage()
  n36 = ScalarParameter() {defaultValue=1.1, group=Blue, parameterName=BlueScale}
  n37 = ScalarParameter() {group=Blue, parameterName=BlueX}
  n38 = ScalarParameter() {group=Blue, parameterName=BlueY}
  n39 = MaterialFunctionCall(
   UVs=Add(
    A=MaterialFunctionCall(UVs=@n35, Texture Scale=@n36) {materialFunction=&ScaleUVsByCenter}.UVs
    B=AppendVector(A=@n37, B=@n38)
   )
   "Rotation Angle (0-1)"=MaterialFunctionCall!(Float 2, Seed, Single value=ScalarParameter() {group=Random, parameterName=Seed}) {materialFunction=/Game/Catalog/BackgroundMaterialLibrary/CoreMaterials/MaterialFunctions/MF_UI_Random}.Single value
  ) {materialFunction=&CustomRotator}
  NamedRerouteDeclaration(None=ComponentMask(None=MaterialFunctionCall(UVs=@n39.Rotated Values, Texture Object=@n31, Grayscale Flow Map=@n18.RGB, Flow Direction=@n32, Flow Intensity=@n33, Flow Speed=@n34) {materialFunction=&MF_UI_GrayscaleFlowMap_New}.Distorted Texture) {b=true}) {name=Blue, nodeColor=<rgba 0 0.0329 1 1>}
  NamedRerouteUsage()
 ]
}
